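# Idle Niagara effect for the OctopusChewTwinkleUniversal pickaxe cosmetic, layering two ribbon emitters with small sprite drops.
# Ribbon_BoneFlow shapes ribbons along the skeletal mesh's sockets using FN_BoneFlow, with binormal facing and curve-driven width.
# Ribbon_Tendrils001 grows wavy, curl-noise tendrils, seeding randomness by ribbon ID so each strand stays coherent.
# Drops spawn at six socket groups, shrink under drag and point attraction, and live shorter as owner speed rises.
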NiagaraSystem /BRCosmetics/Weapons/FORT_Melee/Pickaxe_OctopusChewTwinkleUniversal/FX/NS_Pickaxe_OctopusChewTwinkleUniversal_Idle
dependencies {
 usedDataInterfaces [
  &NiagaraDataInterfaceCurve
  &NiagaraDataInterfaceSkeletalMesh
  &NiagaraDataInterfaceSocketReader
 ]
 usedDynamicInputs [&SimulationPosition]
 usedModules [
  &self:SystemSpawnScript.SystemScriptSource.SystemScriptGraph.NiagaraNodeAssignment_0.SetVariables_#1
  &self:SystemSpawnScript.SystemScriptSource.SystemScriptGraph.NiagaraNodeAssignment_1.SetVariables_#2
  &SystemState
 ]
}
emitters {
 Drops {
  emitterUpdateScript [
   &EmitterState()
   &SpawnRate(SpawnRate=4, Use Spawn Probability=false)
  ]
  particleSpawnScript [
   &FN_SetSpawnGroup(Number of Spawn Groups=6)
   &InitializeParticle(
    Color=<rgba 0.276 0.276 0.276 1>
    Lifetime=&FloatFromCurve(
     FloatCurve=DataInterface() {propertyValues {Curve {keys [<at 0 1>,<at 1 0>]}}}
     CurveIndex=&NormalizeFloat(Value=&VectorLength(Vector=$Engine.Owner.Velocity), Min=0, Max=650)
     Scale Curve=0.8
    )
    Sprite Size=&Vector2DFromFloat(Value=&RandomRangeFloat(Minimum=1, Maximum=3.5))
    Sprite Size Mode=::NewEnumerator4
    UsePositionOffset=false
    Write Material Random=true
   ) {module=InitializeParticle001}
   &SocketLocation(
    SamplePosOffset=false
    Socket Index=$Particles.SpawnGroup
    Socket Reader=$System.Socket Reader
    Socket Sampling Mode=::NewEnumerator3
    VelocitySpeedLimit=false
   )
   /Niagara/Modules/Update/SubUV/SubUVAnimation(SubUV Animation Mode=::NewEnumerator1) {module=SubUVAnimation001}
   &self:AddVelAlongSocketAxis_01(
    SocketAxis:Vector3f=&RandomRangeVector(Minimum=<1 -0.2 -0.2>, Maximum=<1 0.2 0.2>)
    Speed:NiagaraFloat=&RandomRangeFloat(Minimum=-25, Maximum=-90)
   ) {module=AddVelAlongSocketAxis_01}
  ]
  particleUpdateScript [
   &ParticleState()
   &InheritVelocity(
    Module.Limit Speed=true
    Velocity Scale=&VectorFromFloat(Value=&FloatFromCurve(
      FloatCurve=DataInterface() {propertyValues {Curve {keys [{arriveTangent=-1.9956, interpMode=RCIM_Cubic, leaveTangent=-1.9956, tangentMode=RCTM_User, time=0, value=1}, <at 1 0 RCIM_Cubic>]}}}
      Scale Curve=0.1
     ))
   ) {module=InheritVelocity001}
   &Drag(Drag=3, Use Linear Drag=true, Use Rotational Drag=false)
   &PointAttractionForce(
    Attraction Radius=300
    Attraction Strength=10
    Attractor Position:Vector3f=$Particles.SkeletalMeshLocation.SampledPosition
    Falloff Exponent=0.7
    Kill Within Radius=false
    Module.Use Falloff=true
    UseAttractorOffset=false
   )
   &SolveForcesAndVelocity(Clamp Velocity=false, Limit Acceleration=false)
   &ScaleSpriteSize(Scale Factor=&Vector2DFromFloat(Value=&FloatFromCurve(FloatCurve=DataInterface() {propertyValues {Curve {keys [<at 0 1 RCIM_Cubic>, {arriveTangent=-1.5322, interpMode=RCIM_Cubic, leaveTangent=-1.5322, tangentMode=RCTM_User, time=1, value=0}]}}})))
  ]
  rendererClasses [&NiagaraSpriteRendererProperties]
  renderers [{rendererClass=&NiagaraSpriteRendererProperties}]
 }
 Ribbon_BoneFlow {
  emitterUpdateScript [
   &EmitterState(Life Cycle Mode=::NewEnumerator1, Loop Duration=1, UseLoopDelay=false)
   &SpawnRate(SpawnRate=&RandomRangeFloat(Minimum=20, Maximum=30), Use Spawn Probability=false)
  ]
  particleSpawnScript [
   &FN_SetSpawnGroup(Inclusivity=true, Normalize Spawn Group=true, Particles Per Group=10, Pattern=::NewEnumerator0, Set Ribbon ID and Ribbon Link Order=true)
   &InitializeRibbon(
    Lifetime Max=1.5
    Lifetime Min=0.8
    Lifetime Mode:ENiagara_LifetimeMode=::NewEnumerator1
    Ribbon Width=&RandomRangeFloat(Minimum=2, Maximum=1.5)
    UsePositionOffset=false
    Write Material Random=true
   )
  ]
  particleUpdateScript [
   &ParticleState()
   /Game/Effects/Niagara/Modules/BoneFlow/FN_BoneFlow_v4/FN_BoneFlow(
    Angle:NiagaraFloat=&Multiply_Float(A=$Particles.Normalized Spawn Group, B=&RandomRangeFloat(Minimum=-0.6, Maximum=0.6))
    BoneID:NiagaraInt32=&ModuloInt(A=$Particles.SpawnGroup, B=1)
    Height:NiagaraFloat=&Multiply_Float(
     A=$Particles.Normalized Spawn Group
     B=&FloatFromCurve(FloatCurve=DataInterface() {propertyValues {Curve {keys [
          {arriveTangent=0.2329, interpMode=RCIM_Cubic, leaveTangent=0.2329, tangentMode=RCTM_User, time=0, value=1}
          {interpMode=RCIM_Cubic, tangentMode=RCTM_User, time=1, value=1.1}
         ]}}})
    )
    Radius:NiagaraFloat=&Multiply_Float(
     A=$Particles.Normalized Spawn Group
     B=&FloatFromCurve(
      FloatCurve=DataInterface() {propertyValues {Curve {keys [
          {arriveTangent=4.8142, interpMode=RCIM_Cubic, leaveTangent=4.8142, tangentMode=RCTM_User, time=0, value=10}
          {arriveTangent=-26.8012, interpMode=RCIM_Cubic, leaveTangent=-26.8012, tangentMode=RCTM_User, time=0.4, value=10}
          {arriveTangent=-23.2104, interpMode=RCIM_Cubic, leaveTangent=-23.2104, tangentMode=RCTM_User, time=1, value=0}
         ]}}}
      CurveIndex=$Particles.Normalized Spawn Group
      Scale Curve=&RandomRangeFloat(Minimum=2, Maximum=1)
     )
    )
    SkeletalMesh:NiagaraDataInterfaceSkeletalMesh=$System.Skeletal Mesh
    Use Sockets Instead=true
    Write Position Attribute=true
   ) {module=FN_BoneFlow001}
   &ScaleRibbonWidth(Ribbon Width Scale=&FloatFromCurve(
     FloatCurve=DataInterface() {propertyValues {Curve {keys [
         {interpMode=RCIM_Cubic, tangentMode=RCTM_User, time=0, value=0}
         {interpMode=RCIM_Cubic, tangentMode=RCTM_User, time=0.2, value=2}
         {arriveTangent=-8.5417, interpMode=RCIM_Cubic, leaveTangent=-8.5417, tangentMode=RCTM_User, time=0.6, value=0.575}
         {arriveTangent=-2.6171, interpMode=RCIM_Cubic, leaveTangent=-2.6171, tangentMode=RCTM_User, time=0.8, value=1}
         {arriveTangent=-0.2025, interpMode=RCIM_Cubic, leaveTangent=-0.2025, tangentMode=RCTM_User, time=1, value=0}
        ]}}}
     CurveIndex=$Particles.Normalized Spawn Group
     Scale Curve=&FloatFromCurve(
      FloatCurve=DataInterface() {propertyValues {Curve {keys [
          {arriveTangent=2.8989, interpMode=RCIM_Cubic, leaveTangent=2.8989, tangentMode=RCTM_User, time=0.01, value=0}
          {interpMode=RCIM_Cubic, tangentMode=RCTM_User, time=0.4, value=0.95}
          {interpMode=RCIM_Cubic, tangentMode=RCTM_User, time=0.6, value=1}
          {arriveTangent=-1.5059, interpMode=RCIM_Cubic, leaveTangent=-1.5059, tangentMode=RCTM_User, time=0.99, value=0}
         ]}}}
      CurveIndex=$Particles.Age
     )
    ))
   /SetVariables(Particles.RibbonFacing=$Output.FN_BoneFlow001.SplineBinormal) {isSetParameters=true}
   &DynamicMaterialParameters(
    Index 0 Param 1=&FloatFromCurve(FloatCurve=DataInterface() {propertyValues {Curve {keys [<at 0 0 RCIM_Cubic>,<at 0.5 1>,<at 1 0 RCIM_Cubic>]}}})
    Module.Param0WriteEnabled=true
    Module.Param1WriteEnabled=true
    Module.Param2WriteEnabled=false
    Module.Param3WriteEnabled=false
   )
   &SolveForcesAndVelocity(Clamp Velocity=false, Limit Acceleration=false)
  ]
  rendererClasses [&NiagaraRibbonRendererProperties]
  renderers [{rendererClass=&NiagaraRibbonRendererProperties}]
 }
 Ribbon_Tendrils001 {
  emitterUpdateScript [
   &EmitterState(Life Cycle Mode=::NewEnumerator1, Loop Delay=1, Loop Duration=1, UseLoopDelay=false)
   &SpawnRate(SpawnRate=20, Use Spawn Probability=false)
  ]
  particleSpawnScript [
   &PartitionParticles(
    Number Of Partitions=3
    Particles Per Partition=20
    Partition Mode:ENiagara_PartitionMode=::NewEnumerator1
    Write Partition Index to Ribbon Link Order:ENiagara_AttributeSamplingApplyOutput=::NewEnumerator0
    Write Partition to Ribbon ID=::NewEnumerator0
   )
   /SetVariables(Particles.RibbonID=/Niagara/DynamicInputs/NiagaraID/MakeNiagaraIDFromInt(Index:NiagaraInt32=$Emitter.LoopCount, Acquire Tag:NiagaraInt32=0)) {enabled=false, isSetParameters=true}
   &InitializeParticle(
    Lifetime=&RandomRangeFloat(
     Minimum=0.5
     Maximum=1.5
     Randomness Mode=::NewEnumerator1
     Random Seed=&ReturnParticleIDAttributes(Particle ID Attribute=::NewEnumerator1, Niagara ID=$Particles.RibbonID)
     Fixed Random Seed=true
    )
    Position Offset=&AddVectorToPosition(Position=<2 0 0>, Vector=<0 0 0>)
    Position Offset Coordinate Space=::Local
    Ribbon Width=&RandomRangeFloat(
     Minimum=3
     Maximum=4
     Randomness Mode=::NewEnumerator1
     Random Seed=&ReturnParticleIDAttributes(Particle ID Attribute=::NewEnumerator1, Niagara ID=$Particles.RibbonID)
     Fixed Random Seed=true
    )
    Ribbon Width Mode=::NewEnumerator1
    UsePositionOffset=true
    Write Material Random=true
   )
   /SetVariables(Particles.SpawnTime:NiagaraFloat=&FloatFromCurve(FloatCurve=DataInterface() {propertyValues {Curve {keys [<at 0.2 0>,<at 0.8 1>]}}}, CurveIndex=$Emitter.Age)) {isSetParameters=true}
   &AddVelocity(
    Velocity=&Add_Vector(
     A=&MakeVector(
      X=-1
      Y=&Waveform(
       Global Input Position=&RandomRangeFloat(
        Minimum=0
        Maximum=100
        Randomness Mode=::NewEnumerator1
        Random Seed=&ReturnParticleIDAttributes(Particle ID Attribute=::NewEnumerator1, Niagara ID=$Particles.RibbonID)
        Fixed Random Seed=true
       )
       "Waveform [1]"=::NewEnumerator12
       "[1] Frequency"=2
       "[1] Amplitude Min/Max"=<-1 1>
       "Waveform [2] Blend"=::NewEnumerator2
       "Waveform [2]"=::NewEnumerator12
      )
      Z=&Waveform(
       Global Input Position=&RandomRangeFloat(
        Minimum=0
        Maximum=100
        Randomness Mode=::NewEnumerator1
        Random Seed=&ReturnParticleIDAttributes(Particle ID Attribute=::NewEnumerator1, Niagara ID=$Particles.RibbonID)
        Fixed Random Seed=true
       )
       "Waveform [1]"=::NewEnumerator12
       "[1] Frequency"=3
       "[1] Amplitude Min/Max"=<-1 1>
       "Waveform [2] Blend"=::NewEnumerator2
      )
     )
     B=&RandomRangeVector(Minimum=<0 0 0>, Maximum=<0 0 0>)
    )
    Velocity Speed Scale=15
   )
  ]
  particleUpdateScript [
   &ParticleState()
   /Niagara/Modules/Update/Forces/CurlNoiseForce(Noise Frequency=0.3, Noise Strength=20, Pan Noise=true, Pan Noise Field=<0.1 -0.3 0.003>)
   &Drag(Drag=&RandomRangeFloat(Minimum=0.5, Maximum=0.5), Use Linear Drag=true, Use Rotational Drag=true)
   &AccelerationForce(
    Acceleration=&Multiply_VectorByFloat(
     Vector=&RandomRangeVector(Minimum=<-70 0 0>, Maximum=<-80 0 0>)
     Float=&RandomRangeFloat(
      Minimum=0
      Maximum=-0.3
      Randomness Mode=::NewEnumerator1
      Random Seed=&ReturnParticleIDAttributes(Particle ID Attribute=::NewEnumerator1, Niagara ID=$Particles.RibbonID)
      Fixed Random Seed=true
     )
    )
    Coordinate Space=::Local
   )
   &SolveForcesAndVelocity(Clamp Velocity=false, Limit Acceleration=false)
   &DynamicMaterialParameters(
    Index 0 Param 1=1
    Index 0 Param 2=$Particles.MaterialRandom
    Index 0 Param 3=1
    Module.Param0WriteEnabled=true
    Module.Param1WriteEnabled=true
    Module.Param2WriteEnabled=false
    Module.Param3WriteEnabled=true
   ) {module=DynamicMaterialParameters001}
   &ScaleRibbonWidth(Ribbon Width Scale=&Multiply_Float(
     A=&FloatFromCurve(FloatCurve=DataInterface() {propertyValues {Curve {keys [<at 0 0>,<at 0.2 1>,<at 0.7 0>]}}}, Scale Curve=$Particles.SpawnTime)
     B=&FloatFromCurve(FloatCurve=DataInterface() {propertyValues {Curve {keys [<at 0 0>,<at 0.2 1>,<at 0.5 1>,<at 1 0>]}}}, CurveIndex=$Particles.RibbonLinkOrder)
    ))
  ]
  rendererClasses [&NiagaraRibbonRendererProperties]
  renderers [{rendererClass=&NiagaraRibbonRendererProperties}]
 }
}
summary {
 systemName=NS_Pickaxe_OctopusChewTwinkleUniversal_Idle
 emitters [
  {emitterName=Ribbon_BoneFlow, rendererClasses [&NiagaraRibbonRendererProperties]}
  {emitterName=Ribbon_Tendrils001, rendererClasses [&NiagaraRibbonRendererProperties]}
  {emitterName=Drops, rendererClasses [&NiagaraSpriteRendererProperties]}
 ]
 userVariables []
}
