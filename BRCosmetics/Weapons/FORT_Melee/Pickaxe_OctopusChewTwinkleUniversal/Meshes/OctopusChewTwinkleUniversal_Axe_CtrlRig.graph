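# Control Rig for the OctopusChewTwinkleUniversal pickaxe that procedurally wiggles its "oil" bone chains, all parented to the two-handed melee handle.
# One forward-solve sequence, with no variables or inputs, alternates Item Harmonics on individual bones with Chain Harmonics along most chains.
# Per-chain noisy waves drive the motion; pendulum is disabled throughout, and reach is enabled only on oilBack_A and oilBack_G.
# oilBack_B uses only per-bone harmonics, and every chain node has debug drawing on.
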ControlRigBlueprint /BRCosmetics/Weapons/FORT_Melee/Pickaxe_OctopusChewTwinkleUniversal/Meshes/OctopusChewTwinkleUniversal_Axe_CtrlRig
graphs [
 RigVMFunctionLibrary {links [], nodes []}
 RigVMModel {nodes [
   RigUnit_BeginExecution = &RigUnit_BeginExecution()
   /Script/ControlRig.RigUnit_ItemHarmonics(
    RotationOrder=YZX
    Targets=((Item=(Type=Bone,Name="oilA_base"),Ratio=0))
    WaveAmplitude=(X=1,Y=1,Z=1)
    WaveEase=Linear
    WaveFrequency=(X=1,Y=0.6,Z=0.8)
    WaveMaximum=1
    WaveMinimum=0.5
    WaveNoise=(X=2,Y=2,Z=2)
    WaveOffset=(X=0,Y=1,Z=2)
    WaveSpeed=(X=2,Y=2,Z=2)
    ExecutePin=@RigUnit_BeginExecution.ExecutePin
   ) {is=struct RigVMUnitNode}
   > /Script/ControlRig.RigUnit_ChainHarmonicsPerItem(
    ChainRoot=(Type=Bone,Name="oilA_base")
    DrawWorldOffset=(Rotation=(X=0,Y=0,Z=0,W=1),Translation=(X=0,Y=0,Z=0),Scale3D=(X=1,Y=1,Z=1))
    Pendulum=(bEnabled=False,PendulumStiffness=2,PendulumGravity=(X=0,Y=0,Z=0),PendulumBlend=0.75,PendulumDrag=0.98,PendulumMinimum=0,PendulumMaximum=0.1,PendulumEase=Linear,UnwindAxis=(X=0,Y=1,Z=0),UnwindMinimum=0.2,UnwindMaximum=0.05)
    Reach=(bEnabled=false,ReachTarget=(X=0,Y=0,Z=0),ReachAxis=(X=1,Y=0,Z=0),ReachMinimum=-5,ReachMaximum=0,ReachEase=Linear)
    Speed=(X=0.05,Y=0.05,Z=0.4)
    Wave=(bEnabled=True,WaveFrequency=(X=0.3,Y=0.3,Z=0.3),WaveAmplitude=(X=0.5,Y=0.5,Z=0.5),WaveOffset=(X=1,Y=0,Z=0.5),WaveNoise=(X=5,Y=5,Z=5),WaveMinimum=-20,WaveMaximum=-5,WaveEase=SineEaseIn)
    WaveCurve=(EditorCurveData=(Keys=((),(Time=1,Value=1)),DefaultValue=340282346638528859811704183484516925440,PreInfinityExtrap=RCCE_Constant,PostInfinityExtrap=RCCE_Constant),ExternalCurve=None)
    bDrawDebug=True
   ) {is=struct RigVMUnitNode}
   > &RigUnit_ItemHarmonics(
    RotationOrder=YZX
    Targets=((Item=(Type=Bone,Name="oildB_base"),Ratio=0))
    WaveAmplitude=(X=2,Y=0.5,Z=0)
    WaveEase=Linear
    WaveFrequency=(X=1,Y=0.6,Z=0.8)
    WaveMaximum=1
    WaveMinimum=0.5
    WaveNoise=(X=2,Y=2,Z=2)
    WaveOffset=(X=0,Y=1,Z=2)
    WaveSpeed=(X=1,Y=1,Z=1)
   )
   > &RigUnit_ItemHarmonics(
    RotationOrder=YZX
    Targets=((Item=(Type=Bone,Name="oilA_5"),Ratio=0))
    WaveAmplitude=(X=30,Y=70,Z=15)
    WaveEase=Linear
    WaveFrequency=(X=1,Y=0.6,Z=0.8)
    WaveMaximum=1
    WaveMinimum=0.5
    WaveNoise=(X=2,Y=2,Z=2)
    WaveOffset=(X=0,Y=1,Z=2)
    WaveSpeed=(X=1,Y=1,Z=1)
   )
   > &RigUnit_ChainHarmonicsPerItem(
    ChainRoot=(Type=Bone,Name="oildB_base")
    DrawWorldOffset=(Rotation=(X=0,Y=0,Z=0,W=1),Translation=(X=0,Y=0,Z=0),Scale3D=(X=1,Y=1,Z=1))
    Pendulum=(bEnabled=False,PendulumStiffness=2,PendulumGravity=(X=0,Y=0,Z=0),PendulumBlend=0.75,PendulumDrag=0.98,PendulumMinimum=0,PendulumMaximum=0.1,PendulumEase=Linear,UnwindAxis=(X=0,Y=1,Z=0),UnwindMinimum=0.2,UnwindMaximum=0.05)
    Reach=(bEnabled=False,ReachTarget=(X=0,Y=0,Z=0),ReachAxis=(X=1,Y=0,Z=0),ReachMinimum=0,ReachMaximum=0,ReachEase=Linear)
    Speed=(X=0.1,Y=0.05,Z=0)
    Wave=(bEnabled=True,WaveFrequency=(X=1,Y=1,Z=1),WaveAmplitude=(X=2,Y=0.8,Z=0.2),WaveOffset=(X=1,Y=1,Z=2),WaveNoise=(X=8,Y=8,Z=10),WaveMinimum=-1,WaveMaximum=0.9,WaveEase=SineEaseIn)
    WaveCurve=(EditorCurveData=(Keys=((),(Time=1,Value=1)),DefaultValue=340282346638528859811704183484516925440,PreInfinityExtrap=RCCE_Constant,PostInfinityExtrap=RCCE_Constant),ExternalCurve=None)
    bDrawDebug=True
   )
   > &RigUnit_ChainHarmonicsPerItem(
    ChainRoot=(Type=Bone,Name="oilC_base")
    DrawWorldOffset=(Rotation=(X=0,Y=0,Z=0,W=1),Translation=(X=0,Y=0,Z=0),Scale3D=(X=1,Y=1,Z=1))
    Pendulum=(bEnabled=False,PendulumStiffness=2,PendulumGravity=(X=0,Y=0,Z=0),PendulumBlend=0.75,PendulumDrag=0.98,PendulumMinimum=0,PendulumMaximum=0.1,PendulumEase=Linear,UnwindAxis=(X=0,Y=1,Z=0),UnwindMinimum=0.2,UnwindMaximum=0.05)
    Reach=(bEnabled=False,ReachTarget=(X=0,Y=0,Z=0),ReachAxis=(X=1,Y=0,Z=0),ReachMinimum=0,ReachMaximum=0,ReachEase=Linear)
    Speed=(X=0.08,Y=0.08,Z=0.08)
    Wave=(bEnabled=True,WaveFrequency=(X=1,Y=1,Z=1),WaveAmplitude=(X=2,Y=0.1,Z=2),WaveOffset=(X=3,Y=2,Z=2),WaveNoise=(X=10,Y=10,Z=40),WaveMinimum=-0.9,WaveMaximum=0.6,WaveEase=SineEaseIn)
    WaveCurve=(EditorCurveData=(Keys=((),(Time=1,Value=1)),DefaultValue=340282346638528859811704183484516925440,PreInfinityExtrap=RCCE_Constant,PostInfinityExtrap=RCCE_Constant),ExternalCurve=None)
    bDrawDebug=True
   )
   > &RigUnit_ChainHarmonicsPerItem(
    ChainRoot=(Type=Bone,Name="oildD_base")
    DrawWorldOffset=(Rotation=(X=0,Y=0,Z=0,W=1),Translation=(X=0,Y=0,Z=0),Scale3D=(X=1,Y=1,Z=1))
    Pendulum=(bEnabled=False,PendulumStiffness=2,PendulumGravity=(X=0,Y=0,Z=0),PendulumBlend=0.75,PendulumDrag=0.98,PendulumMinimum=0,PendulumMaximum=0.1,PendulumEase=Linear,UnwindAxis=(X=0,Y=1,Z=0),UnwindMinimum=0.2,UnwindMaximum=0.05)
    Reach=(bEnabled=False,ReachTarget=(X=0,Y=0,Z=0),ReachAxis=(X=1,Y=0,Z=0),ReachMinimum=0,ReachMaximum=0,ReachEase=Linear)
    Speed=(X=0.1,Y=0.2,Z=0.1)
    Wave=(bEnabled=True,WaveFrequency=(X=0.5,Y=0.5,Z=0.5),WaveAmplitude=(X=2,Y=0.5,Z=1),WaveOffset=(X=1,Y=0,Z=2),WaveNoise=(X=5,Y=5,Z=5),WaveMinimum=-2,WaveMaximum=0.75,WaveEase=SineEaseIn)
    WaveCurve=(EditorCurveData=(Keys=((),(Time=1,Value=1)),DefaultValue=340282346638528859811704183484516925440,PreInfinityExtrap=RCCE_Constant,PostInfinityExtrap=RCCE_Constant),ExternalCurve=None)
    bDrawDebug=True
   )
   > &RigUnit_ChainHarmonicsPerItem(
    ChainRoot=(Type=Bone,Name="oilBack_A_base")
    DrawWorldOffset=(Rotation=(X=0,Y=0,Z=0,W=1),Translation=(X=0,Y=0,Z=0),Scale3D=(X=1,Y=1,Z=1))
    Pendulum=(bEnabled=False,PendulumStiffness=2,PendulumGravity=(X=0,Y=0,Z=0),PendulumBlend=0.75,PendulumDrag=0.98,PendulumMinimum=0,PendulumMaximum=0.1,PendulumEase=Linear,UnwindAxis=(X=0,Y=1,Z=0),UnwindMinimum=0.2,UnwindMaximum=0.05)
    Reach=(bEnabled=true,ReachTarget=(X=0,Y=0,Z=0),ReachAxis=(X=1,Y=0,Z=0),ReachMinimum=15,ReachMaximum=0,ReachEase=SineEaseOut)
    Speed=(X=0.1,Y=0.1,Z=0.4)
    Wave=(bEnabled=True,WaveFrequency=(X=0.3,Y=0.3,Z=0.3),WaveAmplitude=(X=1,Y=0.1,Z=0.5),WaveOffset=(X=0.5,Y=0.2,Z=0.5),WaveNoise=(X=2,Y=1,Z=3),WaveMinimum=-50,WaveMaximum=-25,WaveEase=SineEaseIn)
    WaveCurve=(EditorCurveData=(Keys=((),(Time=1,Value=1)),DefaultValue=340282346638528859811704183484516925440,PreInfinityExtrap=RCCE_Constant,PostInfinityExtrap=RCCE_Constant),ExternalCurve=None)
    bDrawDebug=True
   )
   > &RigUnit_ItemHarmonics(
    RotationOrder=YZX
    Targets=((Item=(Type=Bone,Name="oilBack_B_base"),Ratio=0))
    WaveAmplitude=(X=0,Y=0.5,Z=1)
    WaveEase=Linear
    WaveFrequency=(X=1,Y=0.6,Z=0.8)
    WaveMaximum=1
    WaveMinimum=-5
    WaveNoise=(X=1,Y=1,Z=1)
    WaveOffset=(X=0,Y=1,Z=2)
    WaveSpeed=(X=0.1,Y=0.1,Z=0.2)
   )
   > &RigUnit_ItemHarmonics(
    RotationOrder=YZX
    Targets=((Item=(Type=Bone,Name="oilBack_B_1"),Ratio=0))
    WaveAmplitude=(X=1,Y=1,Z=1)
    WaveEase=Linear
    WaveFrequency=(X=0,Y=1,Z=1)
    WaveMaximum=1
    WaveMinimum=-5
    WaveNoise=(X=1,Y=2,Z=1)
    WaveOffset=(X=0,Y=1,Z=2)
    WaveSpeed=(X=0.1,Y=0.15,Z=0.2)
   )
   > &RigUnit_ChainHarmonicsPerItem(
    ChainRoot=(Type=Bone,Name="oilBack_C_base")
    DrawWorldOffset=(Rotation=(X=0,Y=0,Z=0,W=1),Translation=(X=0,Y=0,Z=0),Scale3D=(X=1,Y=1,Z=1))
    Pendulum=(bEnabled=False,PendulumStiffness=2,PendulumGravity=(X=0,Y=0,Z=0),PendulumBlend=0.75,PendulumDrag=0.98,PendulumMinimum=0,PendulumMaximum=0.1,PendulumEase=Linear,UnwindAxis=(X=0,Y=1,Z=0),UnwindMinimum=0.2,UnwindMaximum=0.05)
    Reach=(bEnabled=False,ReachTarget=(X=0,Y=0,Z=0),ReachAxis=(X=1,Y=0,Z=0),ReachMinimum=0,ReachMaximum=0,ReachEase=Linear)
    Speed=(X=0.5,Y=0.5,Z=0.5)
    Wave=(bEnabled=True,WaveFrequency=(X=0.1,Y=0.3,Z=0.3),WaveAmplitude=(X=0.3,Y=0.5,Z=0.2),WaveOffset=(X=1.5,Y=0,Z=0.5),WaveNoise=(X=2,Y=2,Z=2),WaveMinimum=-15,WaveMaximum=0.75,WaveEase=SineEaseIn)
    WaveCurve=(EditorCurveData=(Keys=((),(Time=1,Value=1)),DefaultValue=340282346638528859811704183484516925440,PreInfinityExtrap=RCCE_Constant,PostInfinityExtrap=RCCE_Constant),ExternalCurve=None)
    bDrawDebug=True
   )
   > &RigUnit_ChainHarmonicsPerItem(
    ChainRoot=(Type=Bone,Name="oilBack_D_base")
    DrawWorldOffset=(Rotation=(X=0,Y=0,Z=0,W=1),Translation=(X=0,Y=0,Z=0),Scale3D=(X=1,Y=1,Z=1))
    Pendulum=(bEnabled=False,PendulumStiffness=2,PendulumGravity=(X=0,Y=0,Z=0),PendulumBlend=0.75,PendulumDrag=0.98,PendulumMinimum=0,PendulumMaximum=0.1,PendulumEase=Linear,UnwindAxis=(X=0,Y=1,Z=0),UnwindMinimum=0.2,UnwindMaximum=0.05)
    Reach=(bEnabled=False,ReachTarget=(X=0,Y=0,Z=0),ReachAxis=(X=1,Y=0,Z=0),ReachMinimum=0,ReachMaximum=0,ReachEase=Linear)
    Speed=(X=0.6,Y=0.5,Z=0.6)
    Wave=(bEnabled=True,WaveFrequency=(X=0.3,Y=0.3,Z=0.3),WaveAmplitude=(X=0.1,Y=0.3,Z=0.2),WaveOffset=(X=2.5,Y=0,Z=0.5),WaveNoise=(X=5,Y=0,Z=2),WaveMinimum=-30,WaveMaximum=-3,WaveEase=SineEaseIn)
    WaveCurve=(EditorCurveData=(Keys=((),(Time=1,Value=1)),DefaultValue=340282346638528859811704183484516925440,PreInfinityExtrap=RCCE_Constant,PostInfinityExtrap=RCCE_Constant),ExternalCurve=None)
    bDrawDebug=True
   )
   > &RigUnit_ChainHarmonicsPerItem(
    ChainRoot=(Type=Bone,Name="oilBack_E_base")
    DrawWorldOffset=(Rotation=(X=0,Y=0,Z=0,W=1),Translation=(X=0,Y=0,Z=0),Scale3D=(X=1,Y=1,Z=1))
    Pendulum=(bEnabled=False,PendulumStiffness=2,PendulumGravity=(X=0,Y=0,Z=0),PendulumBlend=0.75,PendulumDrag=0.98,PendulumMinimum=0,PendulumMaximum=0.1,PendulumEase=Linear,UnwindAxis=(X=0,Y=1,Z=0),UnwindMinimum=0.2,UnwindMaximum=0.05)
    Reach=(bEnabled=False,ReachTarget=(X=0,Y=0,Z=0),ReachAxis=(X=1,Y=0,Z=0),ReachMinimum=0,ReachMaximum=0,ReachEase=Linear)
    Speed=(X=0.3,Y=0.4,Z=0.3)
    Wave=(bEnabled=True,WaveFrequency=(X=1,Y=0,Z=1),WaveAmplitude=(X=0.5,Y=0.5,Z=0.5),WaveOffset=(X=0.5,Y=0,Z=0.1),WaveNoise=(X=1,Y=1,Z=2),WaveMinimum=-30,WaveMaximum=-10,WaveEase=SineEaseIn)
    WaveCurve=(EditorCurveData=(Keys=((),(Time=1,Value=1)),DefaultValue=340282346638528859811704183484516925440,PreInfinityExtrap=RCCE_Constant,PostInfinityExtrap=RCCE_Constant),ExternalCurve=None)
    bDrawDebug=True
   )
   > &RigUnit_ChainHarmonicsPerItem(
    ChainRoot=(Type=Bone,Name="oilBack_F_base")
    DrawWorldOffset=(Rotation=(X=0,Y=0,Z=0,W=1),Translation=(X=0,Y=0,Z=0),Scale3D=(X=1,Y=1,Z=1))
    Pendulum=(bEnabled=false,PendulumStiffness=2,PendulumGravity=(X=0,Y=0,Z=0),PendulumBlend=0.75,PendulumDrag=0.98,PendulumMinimum=0,PendulumMaximum=0.1,PendulumEase=Linear,UnwindAxis=(X=0,Y=1,Z=0),UnwindMinimum=0.2,UnwindMaximum=0.05)
    Reach=(bEnabled=false,ReachTarget=(X=0,Y=0.5,Z=0),ReachAxis=(X=0,Y=10,Z=0),ReachMinimum=-1,ReachMaximum=1,ReachEase=SineEaseIn)
    Speed=(X=0.3,Y=0.2,Z=0.3)
    Wave=(bEnabled=True,WaveFrequency=(X=1,Y=0.1,Z=0.5),WaveAmplitude=(X=0.1,Y=0.3,Z=0.1),WaveOffset=(X=-5,Y=3,Z=2),WaveNoise=(X=5,Y=5,Z=3),WaveMinimum=-5,WaveMaximum=-2,WaveEase=SineEaseIn)
    WaveCurve=(EditorCurveData=(Keys=((),(Time=1,Value=1)),DefaultValue=340282346638528859811704183484516925440,PreInfinityExtrap=RCCE_Constant,PostInfinityExtrap=RCCE_Constant),ExternalCurve=None)
    bDrawDebug=True
   )
   > &RigUnit_ItemHarmonics(
    RotationOrder=YZX
    Targets=((Item=(Type=Bone,Name="oilBack_G_base"),Ratio=0))
    WaveAmplitude=(X=1,Y=1,Z=1)
    WaveEase=Linear
    WaveFrequency=(X=1,Y=0.6,Z=0.8)
    WaveMaximum=1
    WaveMinimum=0.5
    WaveNoise=(X=1,Y=1,Z=1)
    WaveOffset=(X=0,Y=1,Z=2)
    WaveSpeed=(X=1,Y=1,Z=1)
   )
   > &RigUnit_ChainHarmonicsPerItem(
    ChainRoot=(Type=Bone,Name="oilBack_G_base")
    DrawWorldOffset=(Rotation=(X=0,Y=0,Z=0,W=1),Translation=(X=0,Y=0,Z=0),Scale3D=(X=1,Y=1,Z=1))
    Pendulum=(bEnabled=False,PendulumStiffness=2,PendulumGravity=(X=0,Y=0,Z=0),PendulumBlend=0.75,PendulumDrag=0.98,PendulumMinimum=0,PendulumMaximum=0.1,PendulumEase=Linear,UnwindAxis=(X=0,Y=1,Z=0),UnwindMinimum=0.2,UnwindMaximum=0.05)
    Reach=(bEnabled=true,ReachTarget=(X=1,Y=1,Z=1),ReachAxis=(X=1,Y=0,Z=0),ReachMinimum=0,ReachMaximum=-2,ReachEase=SineEaseOut)
    Speed=(X=0.2,Y=0.15,Z=0.15)
    Wave=(bEnabled=True,WaveFrequency=(X=0.5,Y=0.3,Z=0.3),WaveAmplitude=(X=0.6,Y=0.4,Z=0.4),WaveOffset=(X=1,Y=0,Z=0.5),WaveNoise=(X=1.5,Y=5,Z=10),WaveMinimum=-20,WaveMaximum=-1,WaveEase=SineEaseIn)
    WaveCurve=(EditorCurveData=(Keys=((),(Time=1,Value=1)),DefaultValue=340282346638528859811704183484516925440,PreInfinityExtrap=RCCE_Constant,PostInfinityExtrap=RCCE_Constant),ExternalCurve=None)
    bDrawDebug=True
   )
  ]}
]
hierarchy [
 b_AR_Root {loc=<0 0 0>, rot=<0 0 0 1>}
 Melee_TwoHanded_Handle {parent=b_AR_Root, loc=<0 0 0>, rot=<0 0 0.7071 -0.7071>}
 oilBack_A_base {parent=Melee_TwoHanded_Handle, loc=<-0.5555 1.2793 -9.51>, rot=<0.2304 -0.2906 -0.6446 -0.6685>}
 oilBack_A_1 {parent=oilBack_A_base, loc=<3.5494 0 0>, rot=<0.0437 0.298 -0.0137 -0.9535>}
 oilBack_A_2 {parent=oilBack_A_1, loc=<4.5852 0 0>, rot=<0 0.184 0 -0.9829>}
 oilA_base {parent=Melee_TwoHanded_Handle, loc=<-1.7847 -14.643 -4.0035>, rot=<-0.1569 0.0514 0.4448 -0.8803>}
 oilA_1 {parent=oilA_base, loc=<4.5141 0 0>, rot=<-0.0392 0.1214 -0.115 -0.9851>}
 oilA_2 {parent=oilA_1, loc=<3.8946 0 0>, rot=<0.1782 -0.0789 0.5159 -0.8342>}
 oilA_4 {parent=oilA_2, loc=<4.1131 0.3828 0.6145>, rot=<0.2421 0.1457 -0.0774 -0.9561>}
 oilA_5 {parent=oilA_4, loc=<3.7727 0 0>, rot=<0.1955 -0.2488 -0.0475 -0.9475>}
 oilA_3 {parent=oilA_2, loc=<1.3427 1.2395 2.9773>, rot=<0.7694 -0.0631 0.6205 0.1381>}
 oildB_base {parent=Melee_TwoHanded_Handle, loc=<0 -36.7043 -1.9229>, rot=<0.0782 0.1493 0.4686 -0.8672>}
 oildB_1 {parent=oildB_base, loc=<6.0956 0 0>, rot=<-0.0009 -0.0002 0.2452 -0.9695>}
 oildB_2 {parent=oildB_1, loc=<5.0949 0 0>, rot=<0.1996 -0.3368 0.1137 -0.9131>}
 oildB_3 {parent=oildB_2, loc=<3.8034 0 0>, rot=<-0.2088 0.4476 -0.0844 -0.8654>}
 oildB_4 {parent=oildB_3, loc=<2.8233 0 0>, rot=<0.0821 -0.3484 0.0306 -0.9333>}
 oilC_base {parent=Melee_TwoHanded_Handle, loc=<-2.7503 -38.827 3.5089>, rot=<0.1658 0.1779 0.6319 -0.7359>}
 oilC_1 {parent=oilC_base, loc=<7.1855 0 0>, rot=<-0.1282 -0.1678 -0.0387 -0.9767>}
 oilC_2 {parent=oilC_1, loc=<6.0252 0 0>, rot=<0.0308 -0.2192 -0.0206 -0.975>}
 oilC_3 {parent=oilC_2, loc=<5.5947 0 0>, rot=<-0.0283 -0.0064 0.2193 -0.9752>}
 oilC_4 {parent=oilC_3, loc=<4.4588 0 0>, rot=<0 0.0332 0 -0.9994>}
 oildD_base {parent=Melee_TwoHanded_Handle, loc=<-2.0314 -20.7519 -1.6009>, rot=<0.1473 0.1166 0.8202 -0.5403>}
 oildD_1 {parent=oildD_base, loc=<4.1937 0 0>, rot=<0.1424 0.2419 -0.0771 -0.9567>}
 oildD_2 {parent=oildD_1, loc=<3.5936 0 0>, rot=<0.0814 0.2724 -0.1425 -0.9481>}
 oilBack_B_base {parent=Melee_TwoHanded_Handle, loc=<0.218 -4.9598 -9.2211>, rot=<-0.0205 0.01 0.8992 0.4369>}
 oilBack_B_1 {parent=oilBack_B_base, loc=<5.1775 0 0>, rot=<0.0374 0.099 0.3516 -0.9301>}
 oilBack_C_base {parent=Melee_TwoHanded_Handle, loc=<-4.0214 -10.1543 -0.0605>, rot=<0.1023 0.2416 -0.7922 -0.551>}
 oilBack_C_1 {parent=oilBack_C_base, loc=<6.1082 0 0>, rot=<0 -0.2142 0 -0.9768>}
 oilBack_C_2 {parent=oilBack_C_1, loc=<4.1629 0 0>, rot=<0.1089 -0.2017 -0.0226 -0.9731>}
 oilBack_D_base {parent=Melee_TwoHanded_Handle, loc=<-0.8297 0.8284 10.1531>, rot=<-0.2462 0.0676 0.9324 0.2559>}
 oilBack_D_1 {parent=oilBack_D_base, loc=<4.3411 0 0>, rot=<-0.034 -0.1844 0.5274 -0.8286>}
 oilBack_F_base {parent=Melee_TwoHanded_Handle, loc=<8.1421 -2.5271 0.759>, rot=<0.1935 -0.1908 -0.6853 -0.6757>}
 oilBack_F_1 {parent=oilBack_F_base, loc=<4.4898 0 0>, rot=<0 0.1288 0 -0.9917>}
 oilBack_G_base {parent=Melee_TwoHanded_Handle, loc=<3.4456 -1.805 -7.7484>, rot=<0.0814 0.0925 -0.5169 -0.8471>}
 oilBack_G_1 {parent=oilBack_G_base, loc=<4.3369 0 0>, rot=<0 0 -0.282 -0.9594>}
 oilBack_G_2 {parent=oilBack_G_1, loc=<4.2089 0 0>, rot=<-0.1407 -0.2733 -0.0404 -0.9507>}
 oilBack_E_base {parent=Melee_TwoHanded_Handle, loc=<1.5692 -8.1909 3.5463>, rot=<-0.075 0.135 -0.48 -0.8636>}
 oilBack_E_1 {parent=oilBack_E_base, loc=<7.6743 0 0>, rot=<0.0991 -0.6196 -0.0831 -0.7742>}
 oilBack_E_2 {parent=oilBack_E_1, loc=<5.7043 0 0>, rot=<-0.1485 0.3503 -0.3609 -0.8515>}
 oilBack_E_3 {parent=oilBack_E_2, loc=<3.9938 0 0>, rot=<0 0.5836 0 -0.812>}
]
